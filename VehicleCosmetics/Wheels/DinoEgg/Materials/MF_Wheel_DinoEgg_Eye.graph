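# Material function that shades the animated, color-tinted eyes of the DinoEgg wheel cosmetic and outputs material attributes.
# Two gradient samples panned over time at zero-tiled UVs jitter the whole eye texture, which BumpOffset gives parallax depth.
# Squared CustomColor tints the eye and reflection-vector swirls, then a normal-based view rim, rotated-normal fake shading and a radial mask shape the result.
# That result drives base color and scaled emissive, alongside normal, metallic and roughness.
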MaterialFunction /VehicleCosmetics/Wheels/DinoEgg/Materials/MF_Wheel_DinoEgg_Eye
graph {nodes [
  n1 = Transform(None=ReflectionVectorWS())
  n2 = TextureSample(UVs=Multiply(A=ComponentMask(None=@n1) {g=true, r=true}, B=Constant2Vector() {g=0.25, r=0.25})) {constCoordinate=1, texture=/VehicleCosmetics/Wheels/_SharedTextures/FX/Swirls_D}
  n3 = VectorParameter() {group=2_1_Eyes Properties, parameterName=CustomColor, defaultValue=<rgba 0.9823 0.7991 0.0666 1>}
  n4 = Multiply(A=@n3.RGB, B=@n3.RGB)
  n5 = TextureCoordinate() {coordinateIndex=1, uTiling=0, vTiling=0}
  n6 = Multiply(A=ScalarParameter() {defaultValue=5, group=2_1_Eyes Properties, parameterName=JitterAmount}, B=Time())
  n7 = TextureSample(UVs=Panner(Coordinate=@n5, Time=@n6) {speedX=0.3, speedY=0.1}) {texture=/VehicleCosmetics/Wheels/_SharedTextures/FX/MowGradient01}
  n8 = TextureSample(UVs=Panner(Coordinate=@n5, Time=@n6) {speedX=-0.3, speedY=-0.125}) {texture=&MowGradient01}
  n9 = MaterialFunctionCall(R=Saturate(None=Multiply(A=@n7.RGB) {constB=0.5}), G=Saturate(None=Multiply(A=@n8.RGB) {constB=0.5})) {materialFunction=&AppendMany}
  n10 = Add(
   A=TextureCoordinate() {coordinateIndex=1}
   B=Multiply(
    A=ConstantBiasScale(None=@n9.RG) {bias=-0.8, scale=0.1}
    B=ScalarParameter() {defaultValue=1, group=2_1_Eyes Properties, parameterName=Animation}
   )
  )
  n11 = Constant2Vector() {g=-0.9, r=-1.95}
  n12 = TextureSampleParameter2D(UVs=BumpOffset(Coordinate=Add(A=@n10, B=@n11), Height=Constant() {r=1}) {heightRatio=0.25}) {group=1_Input Textures, parameterName=Eye_DiffuseTetxure, texture=/VehicleCosmetics/Wheels/DinoEgg/Textures/T_Rim_DinoEgg_Eyes_BC}
  n13 = Multiply(A=@n4, B=Desaturation(None=@n12.RGB, Fraction=Constant())) {constA=2}
  n14 = LinearInterpolate(A=Add(A=Multiply(A=@n2.RGB, B=@n13) {constA=2}, B=@n13), B=@n4, Alpha=Multiply(B=@n12.G) {constA=2})
  n15 = TextureSampleParameter2D() {
   constCoordinate=1
   group=1_Input Textures
   parameterName=Eye_NormalTetxure
   samplerType=SAMPLERTYPE_Normal
   texture=/VehicleCosmetics/Wheels/DinoEgg/Textures/T_Rim_DinoEgg_Eyes_N
  }
  n16 = DotProduct(A=Multiply(A=@n15.RGB, B=Constant3Vector() {constant=<rgba 0.05 0.05 1 0>}), B=OneMinus(None=CameraVectorWS()))
  n17 = OneMinus(None=Saturate(None=Multiply(A=@n16) {constB=5}))
  n18 = Multiply(A=@n17, B=@n17) {constB=5}
  n19 = TextureSampleParameter2D(UVs=MaterialFunctionCall(UVs=TextureCoordinate() {coordinateIndex=1}, Rotation Center=Constant2Vector() {g=0.5, r=0.5}, "Rotation Angle (0-1)"=Constant() {r=-0.25}) {materialFunction=&CustomRotator}.Rotated Values) {
   constCoordinate=1
   group=1_Input Textures
   parameterName=Eye_NormalTetxure
   samplerType=SAMPLERTYPE_Normal
   texture=&T_Rim_DinoEgg_Eyes_N
  }
  n20 = ConstantBiasScale(None=ComponentMask(None=Transform(None=@n19.RGB)) {g=true}) {bias=1.2, scale=0.176}
  n21 = TextureSample() {constCoordinate=1, texture=/VehicleCosmetics/Wheels/_SharedTextures/FX/Radial_Generic_01_Pack}
  n22 = Multiply(A=Multiply(A=Add(A=@n14, B=@n18), B=Saturate(None=@n20)), B=Multiply(A=@n21.G, B=@n21.G))
  n23 = SetMaterialAttributes!(
   MaterialAttributes
   BaseColor=@n22
   Normal=@n15.RGB
   Metallic=ScalarParameter() {defaultValue=0.02, group=2_1_Eyes Properties, parameterName=Eye_Metallic}
   Roughness=ScalarParameter() {defaultValue=0.15, group=2_1_Eyes Properties, parameterName=Eye_Roughness}
   EmissiveColor=Multiply(A=@n22, B=ScalarParameter() {group=2_1_Eyes Properties, parameterName=Eye_EmissiveStrength})
  ) {attributeSetTypes [
    #1
    #2
    #3
    #4
    #5
   ]}
  FunctionOutput(None=@n23)
 ]}
