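# Surface material for the DinoEgg vehicle wheel cosmetic, set up for skeletal meshes and the mesh deformer.
# The simple wheel parent function supplies the base material attributes, and a DinoEgg eye function is blended in through the blue channel of the rim RGBA mask.
# That result blends again with the parent output, then passes through the shared vehicle damage and de-rez transition functions.
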Material /VehicleCosmetics/Wheels/DinoEgg/Materials/M_Wheel_DinoEgg
graph {
 properties {
  bAutomaticallySetUsageInEditor=false
  bUseMaterialAttributes=true
  bUsedWithMeshDeformer=true
  bUsedWithSkeletalMesh=true
  usedShadingModels=Default Lit
 }
 nodes [
  n1 = MaterialFunctionCall() {materialFunction=&MF_WheelParent_Simple}
  n2 = BlendMaterialAttributes(
   A=@n1.MaterialAttributes
   B=MaterialFunctionCall() {materialFunction=&/MF_Wheel_DinoEgg_Eye}.Result
   Alpha=ComponentMask(None=@n1.RimRGBAMask) {b=true}
  )
  MaterialFunctionCall(In=MaterialFunctionCall(Material=BlendMaterialAttributes(A=@n2, B=@n1.MaterialAttributes, Alpha=@n1.MaterialAttributes)) {materialFunction=&MF_BaseVehicle_Damage}.Result) {materialFunction=&MF_VehicleDeRez_Transition}
 ]
}
